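# Post-sim AnimBlueprint that poses and adds trailing motion to the LycheeNickel cosmetic tail.
# Chained additive bone rotations curl the tail, a Trail controller adds lag, then the base joint transform is overridden.
# Each update sets trail alpha from skydive, jump and fall states, and relaxation speed from owner direction.
# It also computes Tail Alpha and detects OboeThorn emote montages, but the AnimGraph reads neither Tail Alpha nor OboeThornEmote.
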AnimBlueprint /BRCosmetics/Accessories/FORT_Tails/Tail_M_MED_LycheeNickel/Meshes/M_MED_LycheeNickel_Tail_PostSim_AnimBP
target_skeleton=/Game/Accessories/FORT_Tails/Common/Fortnite_Base_Tail_Skeleton
graphs [
 AnimGraph {nodes [
   n1 = "Transform (Modify) Bone - Bone: C_Tail_A_1_Jnt"(ComponentPose=Local To Component(LocalPose=Input Pose()), Alpha=1, Translation="0,0,0", Rotation="40,0,0", Scale="1,1,1") {
    rotationMode=BMM_Additive
    rotationSpace=BCS_BoneSpace
    translationMode=BMM_Additive
    translationSpace=BCS_BoneSpace
    alphaBoolBlend {blendOption=Linear}
    alphaScaleBiasClamp {bInterpResult=true, interpSpeedDecreasing=4, interpSpeedIncreasing=4}
    boneToModify {boneName=C_Tail_A_1_Jnt}
    rotation {pitch=40}
   }
   n2 = "Transform (Modify) Bone - Bone: C_Tail_A_2_Jnt"(ComponentPose=@n1, Alpha=1, Translation="0,0,0", Rotation="40,0,0", Scale="1,1,1") {
    rotationMode=BMM_Additive
    rotationSpace=BCS_BoneSpace
    translationMode=BMM_Additive
    translationSpace=BCS_BoneSpace
    alphaBoolBlend {blendOption=Linear}
    alphaScaleBiasClamp {bInterpResult=true, interpSpeedDecreasing=4, interpSpeedIncreasing=4}
    boneToModify {boneName=C_Tail_A_2_Jnt}
    rotation {pitch=40}
   }
   n3 = "Transform (Modify) Bone - Bone: C_Tail_A_3_Jnt"(ComponentPose=@n2, Alpha=1, Translation="0,0,0", Rotation="40,0,0", Scale="1,1,1") {
    rotationMode=BMM_Additive
    rotationSpace=BCS_BoneSpace
    translationMode=BMM_Additive
    translationSpace=BCS_BoneSpace
    alphaBoolBlend {blendOption=Linear}
    alphaScaleBiasClamp {bInterpResult=true, interpSpeedDecreasing=4, interpSpeedIncreasing=4}
    boneToModify {boneName=C_Tail_A_3_Jnt}
    rotation {pitch=40}
   }
   n4 = "Transform (Modify) Bone - Bone: C_Tail_A_4_Jnt"(ComponentPose=@n3, Alpha=1, Translation="0,0,0", Rotation="40,0,0", Scale="1,1,1") {
    rotationMode=BMM_Additive
    rotationSpace=BCS_BoneSpace
    translationMode=BMM_Additive
    translationSpace=BCS_BoneSpace
    alphaBoolBlend {blendOption=Linear}
    alphaScaleBiasClamp {bInterpResult=true, interpSpeedDecreasing=4, interpSpeedIncreasing=4}
    boneToModify {boneName=C_Tail_A_4_Jnt}
    rotation {pitch=40}
   }
   n5 = "Transform (Modify) Bone - Bone: C_Tail_A_5_Jnt"(ComponentPose=@n4, Alpha=1, Translation="0,0,0", Rotation="0,0,0", Scale="1,1,1") {
    rotationMode=BMM_Additive
    rotationSpace=BCS_BoneSpace
    translationMode=BMM_Additive
    translationSpace=BCS_BoneSpace
    alphaBoolBlend {blendOption=Linear}
    alphaScaleBiasClamp {bInterpResult=true, interpSpeedDecreasing=4, interpSpeedIncreasing=4}
    boneToModify {boneName=C_Tail_A_5_Jnt}
   }
   n6 = "Transform (Modify) Bone - Bone: C_Tail_A_6_Jnt"(ComponentPose=@n5, Alpha=1, Translation="0,0,0", Rotation="-20,0,0", Scale="1,1,1") {
    rotationMode=BMM_Additive
    rotationSpace=BCS_BoneSpace
    translationMode=BMM_Additive
    translationSpace=BCS_BoneSpace
    alphaBoolBlend {blendOption=Linear}
    alphaScaleBiasClamp {bInterpResult=true, interpSpeedDecreasing=4, interpSpeedIncreasing=4}
    boneToModify {boneName=C_Tail_A_6_Jnt}
    rotation {pitch=-20}
   }
   n7 = "Transform (Modify) Bone - Bone: C_Tail_A_7_Jnt"(ComponentPose=@n6, Alpha=1, Translation="0,0,0", Rotation="-20,0,0", Scale="1,1,1") {
    rotationMode=BMM_Additive
    rotationSpace=BCS_BoneSpace
    translationMode=BMM_Additive
    translationSpace=BCS_BoneSpace
    alphaBoolBlend {blendOption=Linear}
    alphaScaleBiasClamp {bInterpResult=true, interpSpeedDecreasing=4, interpSpeedIncreasing=4}
    boneToModify {boneName=C_Tail_A_7_Jnt}
    rotation {pitch=-20}
   }
   n8 = "Transform (Modify) Bone - Bone: C_Tail_A_8_Jnt"(ComponentPose=@n7, Alpha=1, Translation="0,0,0", Rotation="-40,0,0", Scale="1,1,1") {
    rotationMode=BMM_Additive
    rotationSpace=BCS_BoneSpace
    translationMode=BMM_Additive
    translationSpace=BCS_BoneSpace
    alphaBoolBlend {blendOption=Linear}
    alphaScaleBiasClamp {bInterpResult=true, interpSpeedDecreasing=4, interpSpeedIncreasing=4}
    boneToModify {boneName=C_Tail_A_8_Jnt}
    rotation {pitch=-40}
   }
   n9 = "Transform (Modify) Bone - Bone: C_Tail_A_9_Jnt"(ComponentPose=@n8, Alpha=1, Translation="0,0,0", Rotation="-40,0,0", Scale="1,1,1") {
    rotationMode=BMM_Additive
    rotationSpace=BCS_BoneSpace
    translationMode=BMM_Additive
    translationSpace=BCS_BoneSpace
    alphaBoolBlend {blendOption=Linear}
    alphaScaleBiasClamp {bInterpResult=true, interpSpeedDecreasing=4, interpSpeedIncreasing=4}
    boneToModify {boneName=C_Tail_A_9_Jnt}
    rotation {pitch=-40}
   }
   n10 = Trail controller - Bone: C_Tail_A_9_Jnt!(ComponentPose=@n9, Alpha=Get TrailAlpha() -> (TrailAlpha:double), RelaxationSpeedScale:float=Get RelaxSpeed() -> (RelaxSpeed:double)) -> (Pose) {
    bLimitRotation=true
    chainLength=10
    maxDeltaTime=0.0333
    alphaScaleBiasClamp {bInterpResult=true}
    rotationLimits [
     {limitMax=<180 180 180>}
     {limitMax=<180 180 180>}
     {limitMax=<180 180 180>}
     {limitMax=<180 180 180>}
     {limitMax=<180 180 180>}
     {limitMax=<180 180 180>}
     {limitMax=<180 180 180>}
     {limitMax=<180 180 180>}
     {limitMax=<180 180 180>}
     {limitMax=<180 180 180>}
    ]
    rotationOffsets [<0 0 0>, <0 0 0>, <0 0 0>, <0 0 0>, <0 0 0>, <0 0 0>, <0 0 0>, <0 0 0>, <0 0 0>, <0 0 0>]
    trailBone {boneName=C_Tail_A_9_Jnt}
    trailRelaxationSpeed {editorCurveData {keys [<at 0 10 RCIM_Linear>,<at 0.5107 8.1304 RCIM_Linear>,<at 1 5 RCIM_Linear>]}}
   }
   n11 = "Transform (Modify) Bone - Bone: C_Tail_A_Base_Jnt"(ComponentPose=@n10, Alpha=1, Translation="0,0,0", Rotation="73.141,-40.265,49.406", Scale="1,1,1") {
    rotationMode=BMM_Replace
    rotationSpace=BCS_BoneSpace
    translationMode=BMM_Replace
    alphaBoolBlend {blendOption=Linear}
    alphaScaleBiasClamp {bInterpResult=true, interpSpeedDecreasing=5, interpSpeedIncreasing=5}
    boneToModify {boneName=C_Tail_A_Base_Jnt}
    rotation=<rot 73.141 -40.2651 49.406>
   }
   Output Pose(Result=Component To Local(ComponentPose=@n11)) {layerGroup=None}
  ]}
 EventGraph {nodes [
   Event BlueprintUpdateAnimation()
   > n1 = Sequence!(execute) -> (then_0, then_1, then_2:Exec, then_3:Exec)
   > n2 = Branch(Condition=Get bIsSkyDiving())
   > Set TrailAlpha(TrailAlpha:double=0) -> (Output_Get:double)
   n3 = Branch(execute=@n2.else, Condition=OR Boolean(A=Get bIsJumping(), B=Get bIsFalling()))
   > Set TrailAlpha(TrailAlpha=0.25)
   Set TrailAlpha(execute=@n3.else, TrailAlpha=0.5)
   n4 = Branch(execute=@n1.then_2, Condition=OR Boolean!(A=Get bIsMoving2D(), B=Get bIsSprinting(), C:bool=Get bIsTacticalSprinting()) -> (ReturnValue))
   > n5 = Branch(Condition=OR Boolean(A=Get bIsJumping(), B=Get bIsFalling(), C=Get IsOboeThorn() -> (IsOboeThorn:bool)))
   > Set Tail Alpha(Tail Alpha:double=0.15) -> (Output_Get:double)
   Set Tail Alpha(execute=@n5.else, Tail Alpha=0.75)
   n6 = Branch(execute=@n4.else, Condition=OR Boolean(A=Get bIsJumping(), B=Get bIsFalling(), C=Get IsOboeThorn()))
   > Set Tail Alpha(Tail Alpha=0.15)
   Set Tail Alpha(execute=@n6.else, Tail Alpha=1)
   n7 = GetAnimInstance(self=GetOwnerSkeletalMeshComponent())
   Is Valid(exec=@n1.then_3, InputObject=@n7)
   > n8 = Set IsOboeThorn(IsOboeThorn:bool=OR Boolean!(
     A=Montage_IsPlaying(self=@n7, Montage=/BRCosmetics/Animation/Game/MainPlayer/Emotes/OboeThorn/HERO_LycheeNickel/Emote_OboeThorn_CMM_LycheeNickel_M)
     B=Montage_IsPlaying(self=@n7, Montage=/BRCosmetics/Animation/Game/MainPlayer/Emotes/OboeThorn/HERO_LycheeNickel/Emote_OboeThorn_CMM_LycheeNickel_FE_M)
    ) -> (ReturnValue)) -> (Output_Get:bool)
   > n9 = Branch(Condition=@n8)
   > Set OboeThornEmote(OboeThornEmote:double=1) -> (Output_Get:double)
   Set OboeThornEmote(execute=@n9.else, OboeThornEmote=0)
   n10 = GetOwnerDirection!(self:Custom Character Part Anim Instance Object Reference) -> (ReturnValue:float)
   Set RelaxSpeed(execute=@n1.then_1, RelaxSpeed:double=MapRangeClamped(Value="Absolute (Float)"(A=@n10), InRangeA=85, InRangeB=110, OutRangeA=2, OutRangeB=4)) -> (Output_Get:double)
   TryGetPawnOwner()
  ]}
]
variables [
 IsOboeThorn {type=boolean}
 OboeThornEmote {type=number}
 RelaxSpeed {type=number}
 Tail Alpha {type=number}
 TrailAlpha {type=number}
]
